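# Sound cue for the Scholar back bling's left-to-right sound in Athena.
# A single wave player feeds the output with one clip, with no randomization or modulation, at reduced volume.
# It uses the back bling sound class and close-range pet attenuation.
# A concurrency override allows one instance per owner and blocks new plays while it is active.
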SoundCue /Game/Athena/Sounds/BackBlings/Scholar/BackBling_Scholar_LeftToRight_Cue
root=WavePlayer_0
nodes [{id=WavePlayer_0, properties {soundWaveAssetPtr=&/BackBling_Scholar_LeftToRight_01}}]
properties {
 attenuationSettings=&Fort_Pets_Close_Athena_Att
 bOverrideConcurrency=true
 maxDistance=1800.0001
 soundClassObject=/Game/Sounds/FortSoundClassesAndModes/Fort_Backblings
 volumeMultiplier=0.501
 concurrencyOverrides {bLimitToOwner=true, maxCount=1, resolutionRule=PreventNew}
}
